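# Master skeletal-mesh material for the Sahara player outfit, adding a progressive gold reveal with glow.
# Wraps shared character Pre/PostFX and SkinFX functions, overriding base color and emissive.
# SkinFX_Mask channels define the gold area and the threshold Reveal_Progress clips against; gold color comes from a curve-atlas LUT indexed by view angle.
# HDR emissive combines fresnel, view-facing glow, a pre-skinned height mask and a level-transition fade timed from a Blueprint-set start time.
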Material /BRCosmetics/Characters/Player/Male/Medium/Bodies/M_MED_Sahara/Material/M_MED_Sahara_MASTER
graph {
 properties {
  bAutomaticallySetUsageInEditor=false
  bUseMaterialAttributes=true
  bUsedWithMeshDeformer=true
  bUsedWithMorphTargets=true
  bUsedWithSkeletalMesh=true
  pixelDepthOffsetMode=PDOM_Legacy
  refractionMethod=RM_IndexOfRefraction
  shadingModel=MSM_FromMaterialExpression
  subsurfaceProfile=&SS_HeroSkin_02
  usedShadingModels=Default Lit | Subsurface | Clear Coat | Subsurface Profile | Cloth | Eye
 }
 nodes [
  n1 = Time()
  ConstantBiasScale(None=Sine(None=Multiply(A=@n1) {constB=0.25}))
  n2 = MaterialFunctionCall() {materialFunction=&BaseFN_CharacterFunction_PreFX}
  n3 = GetMaterialAttributes(None=MaterialFunctionCall(
    PreskinnedPosition=@n2.Pre skinned Local Position
    Gmap=@n2.Pre skinned Local Position
    MaterialAttributes=@n2.Pre skinned Local Position
    AO=@n2.Pre skinned Local Position
   ) {materialFunction=&CharacterSkinFX}.Result) {attributeGetTypes [#1, #2, #3]}
  n4 = Multiply(A=@n3.Emissive Color, B=ConstantBiasScale(None=NamedRerouteUsage()) {scale=4})
  n5 = NamedRerouteUsage()
  n6 = MaterialFunctionCall(
   AO=@n2.AO
   Distance Blend=@n2.AO
   Fresnel=@n2.AO
   Gmap=@n2.AO
   Pre Skinned Local Position=@n2.AO
   Material=SetMaterialAttributes!(
    MaterialAttributes=@n3.MaterialAttributes
    Base Color=NamedRerouteUsage()
    Curvature or Metallic=@n3.MaterialAttributes
    Emissive Color=Add(A=@n4, B=@n5) {constB=0}
   ) {attributeSetTypes [#1, #2, #3]}
  ) {materialFunction=&BaseFN_CharacterFunction_PostFX}
  MaterialFunctionCall(In=@n6.Result) {materialFunction=&CINE_MF_CharPostFX}
  n7 = TextureSampleParameter2D() {parameterName=SkinFX_Mask}
  n8 = OneMinus(None=@n7.B)
  n9 = ScalarParameter() {group=Sahara, parameterName=Reveal_Progress}
  n10 = OneMinus(None=NamedRerouteDeclaration(None=Saturate(None=@n9)) {name=Reveal_Progress, nodeColor=<rgba 0 1 0.2266 1>})
  n11 = Subtract(A=Saturate(None=Subtract(A=@n8, B=@n10)), B=ScalarParameter() {defaultValue=0.01, group=Sahara, parameterName=Mask_Clip_Amount}) {constB=0.5}
  n12 = Saturate(None=Multiply(A=Ceil(None=@n11), B=Multiply(A=@n8) {constB=0.75}))
  n13 = NamedRerouteUsage()
  NamedRerouteDeclaration(None=Multiply(A=@n12, B=@n13)) {name=Gold_Reveal_Mask, nodeColor=<rgba 0.6875 0 1 1>}
  n14 = VectorParameter() {group=Sahara, parameterName=Gold_Glow_Color, defaultValue=<rgba 0.7812 0.4138 0.022 0>}
  n15 = Saturate(None=NamedRerouteUsage())
  n16 = Saturate(None=Fresnel(ExponentIn=LinearInterpolate(Alpha=@n15) {constA=10, constB=2}) {baseReflectFraction=0, exponent=2.5})
  n17 = NamedRerouteUsage()
  n18 = NamedRerouteUsage()
  n19 = Multiply(A=Multiply(A=Multiply(A=@n16, B=@n17), B=Max(A=@n18) {constB=0.15})) {constB=2}
  n20 = Multiply(A=NamedRerouteUsage(), B=Multiply(A=NamedRerouteUsage(), B=@n14.RGB))
  n21 = NamedRerouteUsage()
  n22 = VertexNormalWS()
  n23 = CameraVectorWS()
  n24 = Max(
   A=Power(Base=DotProduct(A=@n22, B=@n23), Exp=ScalarParameter() {defaultValue=5, group=Sahara, parameterName=Glow_Exp}) {constExponent=8}
   B=Constant() {r=0.075}
  )
  n25 = Multiply(A=ScalarParameter() {defaultValue=0.5, group=Sahara, parameterName=Gold_Reveal_Glow_Intensity}, B=NamedRerouteUsage()) {constA=0.75}
  n26 = PreSkinnedPosition()
  n27 = PreSkinnedLocalBounds()
  n28 = Divide(A=Subtract(A=VertexInterpolator(VS=@n26).PS, B=@n27.Half Extents), B=@n27.Half Extents)
  n29 = NamedRerouteDeclaration(None=Saturate(None=ComponentMask(None=@n28) {b=true})) {name=LowerBodymask, nodeColor=<rgba 0.8984 0 1 1>}
  n30 = Divide(
   A=Subtract(A=Time(), B=ScalarParameter() {defaultValue=-100, group=Sahara, parameterName=TransitionTime_FromBP})
   B=ScalarParameter() {defaultValue=0.75, group=Sahara, parameterName=Level_Transition_Length}
  )
  n31 = ConstantBiasScale(None=Cosine(None=Saturate(None=@n30)))
  n32 = Add(A=OneMinus(None=Saturate(None=@n31)), B=ScalarParameter() {group=Sahara, parameterName=Level_Transition_Debug})
  n33 = NamedRerouteUsage()
  n34 = Multiply(
   A=NamedRerouteDeclaration(None=Multiply(A=@n32, B=@n33)) {name=Reactive_Mask, nodeColor=<rgba 0.4062 1 0 1>}
   B=ScalarParameter() {defaultValue=1.25, group=Sahara, parameterName=Level_Transition_Intensity}
  )
  n35 = Max(A=NamedRerouteUsage()) {constB=0.5}
  n36 = Add(A=Multiply(A=Multiply(A=@n21, B=@n24), B=Multiply(A=@n25, B=@n29)), B=Multiply(A=@n24, B=Multiply(A=@n34, B=@n35)))
  NamedRerouteDeclaration(None=Add(
    A=Multiply(A=@n14.RGB, B=@n19)
    B=MaterialFunctionCall(LDR Input=@n20, HDR Intensity Multiplier=@n36) {materialFunction=&DeriveHDRfromLDR}.RemappedRange
   )) {name=Emissive_Out, nodeColor=<rgba 0.5703 1 0 1>}
  NamedRerouteDeclaration(None=@n7.G) {name=Gold_Area_Mask, nodeColor=<rgba 0.2422 0 1 1>}
  n37 = Add(
   A=@n12
   B=Multiply(A=NamedRerouteUsage(), B=ScalarParameter() {defaultValue=0.4, group=Sahara, parameterName=Gold_Reveal_Mask_Bias_Amount}) {constB=0.4}
  ) {constB=0.5}
  NamedRerouteDeclaration(None=Multiply(A=@n13, B=Saturate(None=@n37))) {name=Gold_Reveal_Mask_WithBias, nodeColor=<rgba 0.1953 1 0 1>}
  n38 = NamedRerouteUsage()
  n39 = Saturate(None=Power(Base=NamedRerouteUsage(), Exp=Constant() {r=0.25}) {constExponent=5})
  n40 = Subtract(
   A=ScalarParameter() {defaultValue=0.75, group=Sahara, parameterName=Base_Color_Gold_Bias}
   B=Multiply(A=OneMinus(None=@n39), B=Constant() {r=0.65}) {constB=1.5}
  )
  NamedRerouteDeclaration(None=Saturate(None=Add(A=@n38, B=@n40) {constB=0.4})) {name=Base_Color_Biased, nodeColor=<rgba 1 0.8906 0 1>}
  NamedRerouteDeclaration(None=@n3.Base Color) {name=BaseColor_In, nodeColor=<rgba 0.0078 0 1 1>}
  NamedRerouteDeclaration(None=@n2.AO) {name=AO_In, nodeColor=<rgba 0 0.3672 1 1>}
  n41 = CameraVectorWS()
  n42 = Normalize(VectorInput=MaterialFunctionCall(Normal=PixelNormalWS(), Flatness=Constant() {r=0.25}) {materialFunction=&FlattenNormal}.Result)
  n43 = CurveAtlasRowParameter(CurveTime=Saturate(None=DotProduct(A=@n41, B=@n42))) {
   atlas=/BRCosmetics/Characters/Player/Male/Medium/Bodies/M_MED_Sahara/Texture/Sahara_CurveAtlas
   curve=/BRCosmetics/Characters/Player/Male/Medium/Bodies/M_MED_Sahara/Texture/Sahara_Gold_Curve
   group=Sahara
   parameterName=Sahara_Gold_LUT
  }
  n44 = Multiply(A=Saturate(None=NamedRerouteUsage()), B=Multiply(A=NamedRerouteUsage(), B=@n43)) {constB=0.1}
  n45 = Multiply(A=@n44) {constB=0.8}
  n46 = NamedRerouteUsage()
  NamedRerouteDeclaration(None=LinearInterpolate(A=NamedRerouteUsage(), B=LinearInterpolate(A=@n45, B=@n44, Alpha=@n46), Alpha=NamedRerouteUsage())) {name=BaseColor_Out, nodeColor=<rgba 0 0.0625 1 1>}
 ]
}
