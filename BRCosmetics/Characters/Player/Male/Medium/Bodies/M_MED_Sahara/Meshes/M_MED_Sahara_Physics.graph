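# Physics asset for the M_MED_Sahara male medium cosmetic body, covering its dangling left and right chain bones.
# Small capsule bodies sit on a master bone per side, with three links on the left and one on the right.
# Constraints join each link to its parent, allowing moderate swing and almost no twist, so the chains bend without spinning.
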PhysicsAsset /BRCosmetics/Characters/Player/Male/Medium/Bodies/M_MED_Sahara/Meshes/M_MED_Sahara_Physics
bodies {
 Dyn_Chain_l1 [{shapeType=Capsule, length=2, radius=1.5, shapeName=Dyn_Chain_l1_capsule, center=<1 0 0>, rotation=<rot 90 0 0>}]
 Dyn_Chain_l2 [{shapeType=Capsule, length=2, radius=1.5, shapeName=Dyn_Chain_l2_capsule, center=<1 0 0>, rotation=<rot 90 0 0>}]
 Dyn_Chain_l3 [{shapeType=Capsule, length=2, radius=1.5, shapeName=Dyn_Chain_l3_capsule, center=<1 0 0>, rotation=<rot 90 0 0>}]
 Dyn_Chain_r1 [{shapeType=Capsule, length=2, radius=1.5, shapeName=Dyn_Chain_r1_capsule, center=<1 0 0>, rotation=<rot 90 0 0>}]
 Dyn_Master_Chain_l [{shapeType=Capsule, length=0.5, radius=0.5, shapeName=Dyn_Master_Chain_l_capsule}]
 Dyn_Master_Chain_r [{shapeType=Capsule, length=0.5, radius=0.5, shapeName=Dyn_Master_Chain_r_capsule}]
}
constraints [
 {bone1Name=Dyn_Chain_l1, bone2Name=Dyn_Master_Chain_l, swing1LimitDegrees=40, swing2LimitDegrees=40, twistLimitDegrees=1}
 {bone1Name=Dyn_Chain_l2, bone2Name=Dyn_Chain_l1, swing1LimitDegrees=40, swing2LimitDegrees=40, twistLimitDegrees=1}
 {bone1Name=Dyn_Chain_l3, bone2Name=Dyn_Chain_l2, swing1LimitDegrees=40, swing2LimitDegrees=40, twistLimitDegrees=1}
 {bone1Name=Dyn_Chain_r1, bone2Name=Dyn_Master_Chain_r, swing1LimitDegrees=40, swing2LimitDegrees=40, twistLimitDegrees=1}
]
